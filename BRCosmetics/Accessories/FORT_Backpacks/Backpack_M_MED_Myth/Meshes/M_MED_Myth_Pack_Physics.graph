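# Physics asset for the M_MED_Myth backpack cosmetic, defining simple collision and a joint for its pack and sword bones.
# The pack bone gets a tiny box body, while a much larger, elongated box covers the sword.
# A single constraint links sword to pack with tight swing and twist limits, restricting the sword to small angular movement relative to the pack.
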PhysicsAsset /BRCosmetics/Accessories/FORT_Backpacks/Backpack_M_MED_Myth/Meshes/M_MED_Myth_Pack_Physics
bodies {
 pack [{shapeType=Box, extentX=0.5, extentY=0.5, extentZ=0.5, center=<0.0004 -1.4488 -2.5476>}]
 sword [{shapeType=Box, extentX=6.158, extentY=57.2235, extentZ=20.1152, center=<1.2305 -4.4202 -0.844>}]
}
constraints [{bone1Name=sword, bone2Name=pack, swing1LimitDegrees=20, swing2LimitDegrees=10, twistLimitDegrees=10}]
